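# Parent character material for the VileSting Point skin, layering a galaxy look on the shared 2023 character parent functions.
# PreFX output feeds MF_GalaxyLook_VileSting, whose emissive gets parameterized hue shift and desaturation.
# A cubed Fresnel blends that emissive toward a fixed bright azure rim before PostFX finishes the attributes.
# Two other galaxy function calls, CrewStyle and Galaxy_v1, have outputs that nothing uses.
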Material /BRCosmetics/Characters/Player/Female/Medium/Bodies/F_MED_VileSting_Crude/Skins/Point/Materials/M_VileSting_Point_Parent
graph {
 properties {
  bAutomaticallySetUsageInEditor=false
  bUseMaterialAttributes=true
  bUsedWithMeshDeformer=true
  bUsedWithMorphTargets=true
  bUsedWithSkeletalMesh=true
  pixelDepthOffsetMode=PDOM_Legacy
  shadingModel=MSM_FromMaterialExpression
  usedShadingModels=Default Lit | Subsurface | Clear Coat | Subsurface Profile | Cloth
 }
 nodes [
  n1 = MaterialFunctionCall() {materialFunction=&MF_CharacterParent_2023_PreFX}
  n2 = GetMaterialAttributes(None=MaterialFunctionCall!(Material=@n1.Result, TechArtMask=@n1.Result) {materialFunction=&/MF_GalaxyLook_VileSting}.Result) {attributeGetTypes [#1]}
  n3 = Desaturation(
   None=MaterialFunctionCall(Hue Shift Percentage=ScalarParameter() {group=Galaxy_AdditionalControl, parameterName=Galaxy_HueShift}, Texture=@n2.Emissive Color) {materialFunction=&HueShift}.Result
   Fraction=ScalarParameter() {group=Galaxy_AdditionalControl, parameterName=Galaxy_Desaturation, sortPriority=33}
  )
  n4 = Multiply(A=Constant3Vector() {constant=<rgba 0 0.4371 1 1>}) {constB=1.2}
  n5 = Power(Base=@n1.Fresnel) {constExponent=3}
  MaterialFunctionCall(
   Material=SetMaterialAttributes(MaterialAttributes=@n2.MaterialAttributes, Emissive Color=LinearInterpolate(A=@n3, B=@n4, Alpha=@n5)) {attributeSetTypes [#1]}
   M=@n1.M
   Pre Skinned Local Position=@n1.M
   Distance Blend=@n1.M
   Fresnel=@n1.M
  ) {materialFunction=&MF_CharacterParent_2023_PostFX}
  MaterialFunctionCall!(Tech Art Mask=@n1.Tech Art Mask, Mask=@n1.Tech Art Mask, In=@n1.Tech Art Mask) {materialFunction=/Game/Global/Materials/Character_Parent_2023/CrewStyle_Parent_2023/Galaxy/MF_CrewStyle_Galaxy_Inst}
  MaterialFunctionCall!(Material=@n1.Result, TechArtMask=@n1.Result) {materialFunction=/Game/Global/Materials/Character_Parent_2023/Galaxy_Parent_2025/MF_Galaxy_v1}
 ]
}
