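# Parent material for the visor of the VileSting Point character skin, built on the 2023 character parent PreFX and PostFX functions using material attributes.
# A Galaxy function layers onto the PreFX result, with its emissive hue-shifted and desaturated by parameters.
# The VileSting_UseGalaxy scalar blends that version over the base before MF_Visor_VileSting processes it and passes it to PostFX.
# The visor function's UVs also drive the PreFX basic emissive UV override.
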Material /BRCosmetics/Characters/Player/Female/Medium/Bodies/F_MED_VileSting_Crude/Skins/Point/Materials/M_VileSting_Point_Visor_Parent
graph {
 properties {
  bUseMaterialAttributes=true
  bUsedWithClothing=true
  bUsedWithMeshDeformer=true
  bUsedWithMorphTargets=true
  bUsedWithNanite=true
  bUsedWithNiagaraMeshParticles=true
  bUsedWithSkeletalMesh=true
  pixelDepthOffsetMode=PDOM_Legacy
  shadingModel=MSM_FromMaterialExpression
  usedShadingModels=Default Lit | Subsurface | Clear Coat | Subsurface Profile | Cloth
 }
 nodes [
  n1 = MaterialFunctionCall(Override Basic Emissive UVs=@n4.UVs) {materialFunction=&MF_CharacterParent_2023_PreFX}
  n2 = GetMaterialAttributes(None=MaterialFunctionCall!(Material=@n1.Result, TechArtMask=@n1.Result) {materialFunction=/Game/Global/Materials/Character_Parent_2023/Galaxy_Parent_2025/MF_Galaxy_v1}.Result) {attributeGetTypes [#1]}
  n3 = Desaturation(
   None=MaterialFunctionCall(Hue Shift Percentage=ScalarParameter() {group=Galaxy_AdditionalControl, parameterName=Galaxy_HueShift}, Texture=@n2.Emissive Color) {materialFunction=&HueShift}.Result
   Fraction=ScalarParameter() {group=Galaxy_AdditionalControl, parameterName=Galaxy_Desaturation, sortPriority=33}
  )
  n4 = MaterialFunctionCall!(
   PreskinnedPosition=@n1.Pre skinned Local Position
   Material=BlendMaterialAttributes(
    A=@n1.Result
    B=SetMaterialAttributes(MaterialAttributes=@n2.MaterialAttributes, Emissive Color=@n3) {attributeSetTypes [#1]}
    Alpha=ScalarParameter() {parameterName=VileSting_UseGalaxy}
   )
   Tech Art Mask=@n1.Pre skinned Local Position
  ) {materialFunction=&/MF_Visor_VileSting}
  MaterialFunctionCall(Material=@n4.Result, M=@n1.M, Pre Skinned Local Position=@n1.M, Distance Blend=@n1.M, Fresnel=@n1.M) {materialFunction=&MF_CharacterParent_2023_PostFX}
 ]
}
